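; Server-side manager actor that sets up the Stalker phone booths in the level and caps how many remain.
; Once playlist data and navigation are ready, it checks a ScalableFloat enable flag, destroying itself if off or waiting briefly before running SetAllBooths.
; SetAllBooths gathers every BP_StalkerPhoneBooth, removes and destroys booths while the count exceeds the configured maximum with no randomization, then flushes net dormancy and prints the count.
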(event EventBeginPlay
  (if (Networking|IsServer)
    (Class|FortAsyncActionCurrentPlaylistReady|PlaylistDataReadyAsync
      (:OnReady
        (Utilities|CheckNavReady
          (:true
            (if (ScalableFloat|GetValueasBool (Variables|Default|GetRow%20BoothsEnabled))
              (Utilities|FlowControl|Delay 2.0)
              (CallFunction|SetAllBooths)
              (else
                (Actor|DestroyActor)))))))))

(event Collision|EventActorBeginOverlap
  (params (OtherActor :type Actor%20Object%20Reference)))

(event EventTick
  (params (DeltaSeconds :type Float%20%28single-precision%29)))

(fn SetAllBooths
  (bind _outboothsarray (Variables|Default|GetOutBoothsArray))
  (bind _outactors (Actor|GetAllActorsOfClass "/FNMStalkerKnife/Gameplay/PhoneBooth/BP_StalkerPhoneBooth.BP_StalkerPhoneBooth_C"))
  (bind _output_get (Variables|Default|SetOutBoothsArray _outactors))
  (for _array_element _outboothsarray
    (if (> (Utilities|Array|Length _output_get) (Variables|Default|GetRow%20BoothsMaxCount))
      (Utilities|Array|RemoveItem _outboothsarray _array_element)
      (Actor|DestroyActor _array_element)))
  (Networking|FlushNetDormancy)
  (Development|PrintString (Utilities|String|Append "Stalker Booths are setup. Num of booths: " (Utilities|String|ToString(Integer) (Utilities|Array|Length _output_get)))))

(fn ConstructionScript)
