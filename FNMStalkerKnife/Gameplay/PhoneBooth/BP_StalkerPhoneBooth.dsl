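; Stalker Knife phone booth that rings for nearby players and is destroyed after one use.
; On playlist load it sets its ring radius or self-destructs if disabled; overlapping non-bot players without the knife or disguise start a timer looping a ring GameplayEffect and sound indicator.
; Interacting stops the ring, and loadout refresh applies a delayed ShadowBomb effect to the player.
; Players holding the knife or the interact tag cannot use it.
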(event Custom|Loop%20RingSFX
  ; Phonebooth Ring SFX Loop
  (if (Utilities|Array|IsNotEmpty (Variables|Default|GetOverlappingPlayerPawns))
    (bind _returnvalue (GameplayEffects|ApplyGameplayEffectToSelf :self (Ability|GetAbilitySystemComponent 0) :GameplayEffectClass "/FNMStalkerKnife/Gameplay/PhoneBooth/GE_PhoneBooth_Ring.GE_PhoneBooth_Ring_C" :Level 0.0))
    (Variables|Default|SetPhoneboothRingHandle _returnvalue)
    (SoundIndicator|ShowSoundIndicatorforDuration (Variables|Default|GetFortSoundIndicator) (Transformation|GetActorLocation) "/FNMStalkerKnife/Assets/Audio/SC_Stalker_MarkedPlayerRing.SC_Stalker_MarkedPlayerRing" 1.0 2.0 0)))

(event EventBeginPlay
  ; Begin Play: Booth initial setup
  (|CallParentFunction)
  (Class|FortAsyncActionCurrentPlaylistReady|PlaylistDataReadyAsync
    (:OnReady
      (if (ScalableFloat|GetValueasBool (Variables|Default|GetRow_StalkerBoothEnabled))
        (Components|Sphere|SetSphereRadius (Variables|Default|GetPhoneRingRadiusCollider) (ScalableFloat|GetValueatLevel (Variables|Default|GetRow_PhoneRingRadius)) true)
        (else
          (Actor|DestroyActor))))))

(event OnComponentBeginOverlap(PhoneRingRadiusCollider)
  (params (OverlappedComponent :type Primitive%20Component%20Object%20Reference) (OtherActor :type Actor%20Object%20Reference) (OtherComp :type Primitive%20Component%20Object%20Reference) (OtherBodyIndex :type Integer) (bFromSweep :type Boolean) (SweepResult :type Hit%20Result%20Structure%20%28by%20ref%29))
  ; PhoneRing Collider Begin Overlap: Trigger Phone Ring SFX Loop
  (bind _asfort_player_pawn_athena (Utilities|Casting|CastToFortPlayerPawnAthena OtherActor))
  (if (not (Pawn|IsBotControlled _asfort_player_pawn_athena))
    (bind _disguiseagid (Variables|Default|GetDisguiseAGID)
      (:then
        (bind _returnvalue (Inventory|GetInventoryItemwithTag (Fortnite|GetFortPlayerControllerfromActor _asfort_player_pawn_athena) (GameplayTags|MakeLiteralGameplayTag "(TagName=\"Granted.StalkerKnife.Equipped\")")))
        (Utilities|IsValid _returnvalue
          (:Is%20Not%20Valid
            (Utilities|IsValid (Inventory|FindExistingItemForDefinition :self (Fortnite|GetFortPlayerControllerfromActor _asfort_player_pawn_athena) :ItemDefinition _disguiseagid :bInStorageVault false)
              (:Is%20Not%20Valid
                (Utilities|Array|Add (Variables|Default|GetOverlappingPlayerPawns) _asfort_player_pawn_athena)
                (if (and (Variables|Default|GetCanPhoneRing) (not (Utilities|Time|IsValidTimerHandle (Variables|Default|GetPhoneRingSFXLoopHandle))))
                  (bind _returnvalue_2 (Utilities|Time|SetTimerbyEvent (EventDispatchers|CreateEvent :!event Custom|Loop%20RingSFX) 4.0 true))
                  (Variables|Default|SetPhoneRingSFXLoopHandle _returnvalue_2)
                  (CallFunction|LoopRingSFX)))))))
      (:else))))

(event OnComponentEndOverlap(PhoneRingRadiusCollider)
  (params (OverlappedComponent :type Primitive%20Component%20Object%20Reference) (OtherActor :type Actor%20Object%20Reference) (OtherComp :type Primitive%20Component%20Object%20Reference) (OtherBodyIndex :type Integer))
  (bind _overlappingplayerpawns (Variables|Default|GetOverlappingPlayerPawns))
  ; PhoneRing Collider End Overlap: Stop SFX Loop and AV indicator
  (bind _asfort_player_pawn_athena (Utilities|Casting|CastToFortPlayerPawnAthena OtherActor))
  (if (not (Pawn|IsBotControlled _asfort_player_pawn_athena))
    (Utilities|Array|RemoveItem _overlappingplayerpawns _asfort_player_pawn_athena)
    (if (Utilities|Array|IsEmpty _overlappingplayerpawns)
      (SoundIndicator|StopSoundIndicatorLooping (Variables|Default|GetFortSoundIndicator))
      (Utilities|Time|ClearandInvalidateTimerbyHandle (Variables|Default|GetPhoneRingSFXLoopHandle))
      (GameplayEffects|RemoveActiveGameplayEffect (Ability|GetAbilitySystemComponent 0) (Variables|Default|GetPhoneboothRingHandle) -1))))

(event EventInteractEnter
  (bind _returnvalue (Ability|GetAbilitySystemComponent 0))
  ; Phonebooth Event InteractEnter: Trigger and stop appropiate SFX
  (|CallParentFunction)
  (Variables|Default|SetCanPhoneRing false)
  (GameplayEffects|RemoveActiveGameplayEffect _returnvalue (Variables|Default|GetPhoneboothRingHandle) -1)
  (GameplayEffects|ApplyGameplayEffectToSelf _returnvalue "/FNMStalkerKnife/Gameplay/GE_StalkerKnife_PhoneboothInteract.GE_StalkerKnife_PhoneboothInteract_C" 0.0 (Class|AthenaFunctionLibrary|MakeSimpleFortEffectContext (Variables|Default|GetLastPawnToInteract)))
  (Development|PrintString "Interacted with Phonebooth!"))

(event EventEndPlay
  (params (EndPlayReason :type EEndPlayReason%20Enum :default "Destroyed"))
  ; Event EndPlay: Cleanup Handles
  (Utilities|Time|ClearandInvalidateTimerbyHandle (Variables|Default|GetPhoneRingSFXLoopHandle))
  (GameplayEffects|RemoveActiveGameplayEffect (Ability|GetAbilitySystemComponent 0) (Variables|Default|GetPhoneboothRingHandle) -1))

(event EventPrepareForLoadoutRefresh
  (params (InteractingPawn :type Fort%20Pawn%20Object%20Reference))
  ; Event PrepareForLoadoutRefresh: Apply VFX to player
  (|CallParentFunction InteractingPawn)
  (Utilities|Time|SetTimerbyEvent (delegate Custom|DelayedFXApply) 2.0))

(event Custom|DelayedFXApply
  ; Event PrepareForLoadoutRefresh: Apply VFX to player
  (GameplayEffects|ApplyGameplayEffectToSelf (Ability|GetAbilitySystemComponent (stale EventPrepareForLoadoutRefresh InteractingPawn)) "/FNMStalkerKnife/Gameplay/PhoneBooth/GE_ShadowBomb.GE_ShadowBomb_C" 0.0 (GameplayEffects|MakeFortEffectContext (Ability|GetFortAbilitySystemComponent (stale EventPrepareForLoadoutRefresh InteractingPawn)) 0 self 0 0)))

(event EventExitBoothAnimations
  ; Event ExitBoothAnimation: Destroy Phonebooth after use
  (|CallParentFunction)
  (Variables|Interaction|SetbAllowInteract false)
  (Development|PrintString "DESTROYING BOOTH")
  (Utilities|FlowControl|Delay 3.0)
  (Actor|DestroyActor))

(fn BlueprintCanInteract
  (params (InteractingPawn :type Fort%20Pawn%20Object%20Reference) (InteractionBeingAttempted :type EInteractionBeingAttempted%20Enum :default "FirstInteraction") (InteractionType :type TInteractionType%20Enum :default "IT_NoInteraction"))
  (Utilities|IsValid (Inventory|FindExistingItemForDefinition :self (Fortnite|GetFortPlayerControllerfromActor InteractingPawn) :ItemDefinition "/FNMStalkerKnife/Gameplay/WID_StalkerKnife.WID_StalkerKnife" :bInStorageVault false)
    (:Is%20Valid
      _)
    (:Is%20Not%20Valid
      (if (not (GameplayTags|HasMatchingGameplayTag InteractingPawn "(TagName=\"Granted.StalkerKnife.Phonebooth.Interact\")"))
        (return (|CallParentFunction InteractingPawn InteractionBeingAttempted InteractionType)))))
  (return false))

(fn BlueprintGetFailedInteractionString
  (params (InteractingPawn :type Fort%20Pawn%20Object%20Reference) (InteractionBeingAttempted :type EInteractionBeingAttempted%20Enum :default "FirstInteraction"))
  (bind _returnvalue (Fortnite|GetFortPlayerControllerfromActor InteractingPawn))
  (bind _returnvalue_2 (Inventory|FindExistingItemForDefinition :self _returnvalue :ItemDefinition "/FNMStalkerKnife/Gameplay/WID_StalkerKnife.WID_StalkerKnife" :bInStorageVault false))
  (bind _output_get (LocalVariables|SetItem%20Def _returnvalue_2))
  (Utilities|IsValid _output_get
    (:Is%20Valid
      (return (Utilities|Text|FormatText (Variables|Default|GetText%20Already%20Obtained) (Item|GetDisplayName _returnvalue_2))))
    (:Is%20Not%20Valid
      (return ""))))

(fn BlueprintUpdateDynamicInteractionOptions
  (params (InteractingPawn :type Fort%20Pawn%20Object%20Reference) (InteractionBeingAttempted :type EInteractionBeingAttempted%20Enum :default "FirstInteraction") (InteractionType :type TInteractionType%20Enum :default "IT_NoInteraction"))
  (bind _outgameplaytags (|CallParentFunction InteractingPawn InteractionBeingAttempted InteractionType))
  (if (not (Interaction|BlueprintCanInteract InteractingPawn InteractionBeingAttempted InteractionType))
    (GameplayTags|AppendGameplayTagContainers _outgameplaytags (GameplayTags|MakeLiteralGameplayTagContainer "(GameplayTags=((TagName=\"HUD.Athena.InteractionOptions.CollapseInteractKey\")))")))
  (return _outgameplaytags))

(fn ConstructionScript
  (|CallParentFunction))
